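# Physics asset for the SolarDusk medium male player body, simulating its dangling necklace against the character's chest.
# Capsules on the upper spine and both pec deform bones form a torso chain joined by constraints with all angular motion locked.
# The necklace and its star pendant hang in sequence from spine_05 on limited-swing, zero-twist constraints, with much tighter first-axis swing than second.
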PhysicsAsset /BRCosmetics/Characters/Player/Male/Medium/Bodies/M_MED_SolarDusk/Meshes/M_MED_SolarDusk_Physics
bodies {
 deform_pec_aim_l [{shapeType=Capsule, length=10.4348, radius=5.1852, center=<7.0547 -3.2767 0.9229>, rotation=<rot 90 13.1541 0>}]
 deform_pec_aim_r [{shapeType=Capsule, length=10.4348, radius=5.1852, center=<-7.0547 3.2767 -0.9229>, rotation=<rot 90 -166.8459 0>}]
 dyn_Necklace [{shapeType=Capsule, length=6.928, radius=4.3484, center=<3.3143 0 3>, rotation=<rot 90 0 90>}]
 dyn_Necklace_star [{shapeType=Capsule, length=9.0913, radius=4.3159, center=<1.434 0 4>, rotation=<rot 90 0 90>}]
 spine_03 [{shapeType=Capsule, length=1, radius=1}]
 spine_04 [{shapeType=Capsule, length=8.3327, radius=13.1778}]
 spine_05 [{shapeType=Capsule, length=10.604, radius=10.6451, center=<-3.4209 0.3673 0>}]
}
constraints [
 {bone1Name=dyn_Necklace, bone2Name=spine_05, swing1LimitDegrees=6, swing2LimitDegrees=20, twistLimitDegrees=0}
 {bone1Name=dyn_Necklace_star, bone2Name=dyn_Necklace, swing1LimitDegrees=2, swing2LimitDegrees=20, twistLimitDegrees=0}
 {
  bone1Name=spine_05
  bone2Name=spine_04
  swing1LimitDegrees=45
  swing1Motion=Locked
  swing2LimitDegrees=45
  swing2Motion=Locked
  twistLimitDegrees=45
  twistMotion=Locked
 }
 {
  bone1Name=deform_pec_aim_l
  bone2Name=spine_04
  swing1LimitDegrees=45
  swing1Motion=Locked
  swing2LimitDegrees=45
  swing2Motion=Locked
  twistLimitDegrees=45
  twistMotion=Locked
 }
 {
  bone1Name=deform_pec_aim_r
  bone2Name=spine_04
  swing1LimitDegrees=45
  swing1Motion=Locked
  swing2LimitDegrees=45
  swing2Motion=Locked
  twistLimitDegrees=45
  twistMotion=Locked
 }
 {
  bone1Name=spine_04
  bone2Name=spine_03
  swing1LimitDegrees=45
  swing1Motion=Locked
  swing2LimitDegrees=45
  swing2Motion=Locked
  twistLimitDegrees=45
  twistMotion=Locked
 }
]
