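# Physics asset for the M_MED_SolarDusk male character body, covering its necklace bones and the upper torso around them.
# Capsules on spine_03 to spine_05 and both pec deform bones are joined by fully locked constraints, forming a rigid chest shape.
# Two two-link chains hang from spine_05, dyn_Necklace to dyn_Necklace_star and necklace_01 to necklace_02, with tight swing limits and little or no twist.
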PhysicsAsset /BRCosmetics/Characters/Player/Male/Medium/Bodies/M_MED_SolarDusk/Meshes/M_MED_SolarDusk_Physics_fe
bodies {
 deform_pec_aim_l [{shapeType=Capsule, length=4.7783, radius=4.9093, center=<5.377 -3.5375 0>, rotation=<rot 90 17.1912 0>}]
 deform_pec_aim_r [{shapeType=Capsule, length=4.7783, radius=4.9093, center=<-5.377 3.5375 0>, rotation=<rot 90 -162.8088 0>}]
 dyn_Necklace [{shapeType=Capsule, length=6.928, radius=4.3484, center=<3.3143 0 3>, rotation=<rot 90 0 90>}]
 dyn_Necklace_star [{shapeType=Capsule, length=9.2647, radius=4.1425, center=<1.434 0 4>, rotation=<rot 90 0 90>}]
 necklace_01 [{shapeType=Capsule, length=6.0834, radius=4.7703, shapeName=necklace_01_capsule, center=<4 0 4>, rotation=<rot 0 0 -88.6334>}]
 necklace_02 [{shapeType=Capsule, length=5.4421, radius=3.6838, shapeName=necklace_02_capsule, center=<5 0 2.3>, rotation=<rot 0 0 90>}]
 spine_03 [{shapeType=Capsule, length=1, radius=1}]
 spine_04 [{shapeType=Capsule, length=8.3327, radius=13.1778}]
 spine_05 [{shapeType=Capsule, length=11.275, radius=9.9741}]
}
constraints [
 {bone1Name=dyn_Necklace, bone2Name=spine_05, swing1LimitDegrees=6, swing2LimitDegrees=20, twistLimitDegrees=0}
 {bone1Name=dyn_Necklace_star, bone2Name=dyn_Necklace, swing1LimitDegrees=2, swing2LimitDegrees=20, twistLimitDegrees=0}
 {
  bone1Name=spine_05
  bone2Name=spine_04
  swing1LimitDegrees=45
  swing1Motion=Locked
  swing2LimitDegrees=45
  swing2Motion=Locked
  twistLimitDegrees=45
  twistMotion=Locked
 }
 {
  bone1Name=deform_pec_aim_l
  bone2Name=spine_04
  swing1LimitDegrees=45
  swing1Motion=Locked
  swing2LimitDegrees=45
  swing2Motion=Locked
  twistLimitDegrees=45
  twistMotion=Locked
 }
 {
  bone1Name=deform_pec_aim_r
  bone2Name=spine_04
  swing1LimitDegrees=45
  swing1Motion=Locked
  swing2LimitDegrees=45
  swing2Motion=Locked
  twistLimitDegrees=45
  twistMotion=Locked
 }
 {
  bone1Name=spine_04
  bone2Name=spine_03
  swing1LimitDegrees=45
  swing1Motion=Locked
  swing2LimitDegrees=45
  swing2Motion=Locked
  twistLimitDegrees=45
  twistMotion=Locked
 }
 {bone1Name=necklace_01, bone2Name=spine_05, swing1LimitDegrees=12, swing2LimitDegrees=20, twistLimitDegrees=1}
 {bone1Name=necklace_02, bone2Name=necklace_01, swing1LimitDegrees=12, swing2LimitDegrees=20, twistLimitDegrees=1}
]
