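# Physics asset for the dangling rope laces on the M_MED_SolarDusk male player body.
# A small Base_Rope sphere is locked to the pelvis and carries the dyn_Rope_01 capsule on a narrow-twist, limited-swing joint.
# The dyn_Rope_02 and dyn_Rope_03 capsules both branch from dyn_Rope_01 rather than forming a single chain.
# Pelvis and thigh capsules, with thigh rotation locked to the pelvis, provide body shapes around the rope.
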PhysicsAsset /BRCosmetics/Characters/Player/Male/Medium/Bodies/M_MED_SolarDusk/Meshes/M_MED_SolarDusk_Physics_laces
bodies {
 Base_Rope [{shapeType=Sphere, radius=0.505, shapeName=Base_Rope_sphere}]
 dyn_Rope_01 [{shapeType=Capsule, length=3.6135, radius=2.2725, center=<3 0 2.4>, rotation=<rot 90 0 0>}]
 dyn_Rope_02 [{shapeType=Capsule, length=3.4596, radius=2.4823, center=<3 0 1>, rotation=<rot 90 0 0>}]
 dyn_Rope_03 [{shapeType=Capsule, length=3.1239, radius=2.818, center=<3 0 1>, rotation=<rot 90 0 0>}]
 pelvis [{shapeType=Capsule, length=11.3754, radius=9.3083, center=<-3.5092 0.1839 0>}]
 thigh_l [{
   shapeType=Capsule
   length=54.88
   radius=10.3894
   shapeName=thigh_l_capsule
   center=<-31.2483 0.1017 -1.6157>
   rotation=<rot 0 -93.805 -89.2752>
  }]
 thigh_r [{
   shapeType=Capsule
   length=54.88
   radius=10.3894
   shapeName=thigh_r_capsule
   center=<31.2483 -0.1017 1.6157>
   rotation=<rot 0 -93.805 90.7248>
  }]
}
constraints [
 {bone1Name=dyn_Rope_02, bone2Name=dyn_Rope_01, swing1LimitDegrees=20, swing2LimitDegrees=20, twistLimitDegrees=1}
 {bone1Name=dyn_Rope_03, bone2Name=dyn_Rope_01, swing1LimitDegrees=20, swing2LimitDegrees=20, twistLimitDegrees=1}
 {
  bone1Name=thigh_l
  bone2Name=pelvis
  swing1LimitDegrees=45
  swing1Motion=Locked
  swing2LimitDegrees=45
  swing2Motion=Locked
  twistLimitDegrees=45
  twistMotion=Locked
 }
 {
  bone1Name=thigh_r
  bone2Name=pelvis
  swing1LimitDegrees=45
  swing1Motion=Locked
  swing2LimitDegrees=45
  swing2Motion=Locked
  twistLimitDegrees=45
  twistMotion=Locked
 }
 {bone1Name=dyn_Rope_01, bone2Name=Base_Rope, swing1LimitDegrees=4, swing2LimitDegrees=28, twistLimitDegrees=1}
 {
  bone1Name=Base_Rope
  bone2Name=pelvis
  swing1LimitDegrees=45
  swing1Motion=Locked
  swing2LimitDegrees=45
  swing2Motion=Locked
  twistLimitDegrees=45
  twistMotion=Locked
 }
]
