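# Control Rig for the face accessory part of the male medium SolarDusk body, over an upper-body skeleton with many hair bones under the head.
# The graph reads the head's current transform relative to spine_05, converts its rotation to Euler angles and remaps the Z angle to a 0–1 range.
# That result feeds nothing and the begin-execution event has no steps after it, so the rig drives no bones as it stands.
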ControlRigBlueprint /BRCosmetics/Characters/Player/Male/Medium/Bodies/M_MED_SolarDusk/Meshes/Parts/M_MED_SolarDusk_FaceAcc_CtrlRig
graphs [
 RigVMFunctionLibrary {links [], nodes []}
 RigVMModel {nodes [
   &RigVMFunction_MathDoubleRemap(
    SourceMaximum=120
    SourceMinimum=90
    TargetMaximum=1
    TargetMinimum=0
    Value=&RigVMFunction_MathQuaternionToEuler(
     RotationOrder=XYZ
     Value=&RigUnit_GetRelativeTransformForItem(Child=(Type=Bone,Name="head"), Parent=(Type=Bone,Name="spine_05"), bChildInitial=False, bParentInitial=False).RelativeTransform.Rotation
    ).Result.Z
   )
   RigUnit_BeginExecution = &RigUnit_BeginExecution()
  ]}
]
hierarchy [
 root {loc=<0 0 0>, rot=<0 0 0 1>}
 pelvis {parent=root, loc=<0 -0.7071 93.1191>, rot=<0.0127 0.707 -0.0127 -0.707>}
 spine_01 {parent=pelvis, loc=<6.7494 -0.6388 0>, rot=<0 0 0.034 -0.9994>}
 spine_02 {parent=spine_01, loc=<7.2518 -0.035 0>, rot=<0 0 0.0238 -0.9997>}
 spine_03 {parent=spine_02, loc=<6.0104 0 0>, rot=<0 0 0.0149 -0.9999>}
 spine_04 {parent=spine_03, loc=<8.1024 0 0>, rot=<0 0 -0.0132 -0.9999>}
 spine_05 {parent=spine_04, loc=<10.0146 0 0>, rot=<0 0 -0.0015 -1>}
 neck_01 {parent=spine_05, loc=<11.3138 0.1529 0>, rot=<0 0 -0.2003 -0.9797>}
 neck_02 {parent=neck_01, loc=<4.5627 0 0>, rot=<0 0 0.0504 -0.9987>}
 head {parent=neck_02, loc=<5.3914 0 0>, rot=<0 0 0.115 -0.9934>}
 Hair_01 {parent=head, loc=<11.8278 -4.1276 -5.1723>, rot=<0.3099 -0.4354 0.2689 -0.8013>}
 Hair_02 {parent=head, loc=<14.8612 -3.5117 -3.2706>, rot=<0.3622 -0.243 0.1889 -0.8798>}
 Hair_03 {parent=head, loc=<7.9212 12.5729 3.4325>, rot=<0.0175 0.596 0.8024 0.025>}
 Hair_04 {parent=head, loc=<11.4305 -7.3509 1.0847>, rot=<0.6938 -0.2909 0.4864 -0.4443>}
 Hair_05 {parent=head, loc=<12.1413 -5.8506 4.7115>, rot=<0.757 -0.2267 0.5061 -0.3457>}
 Hair_06 {parent=head, loc=<7.3272 7.626 -5.2057>, rot=<-0.1266 -0.9763 -0.1757 -0.0035>}
 Hair_07 {parent=head, loc=<4.059 6.5404 6.3182>, rot=<0.0386 0.1194 0.9919 -0.0217>}
 Hair_08 {parent=head, loc=<4.3577 10.1156 5.2671>, rot=<0.0907 0.1184 0.9869 0.0621>}
 Hair_09 {parent=head, loc=<3.5513 3.7132 7.2331>, rot=<0.2605 0.2454 0.9243 0.1327>}
 Hair_10 {parent=head, loc=<1.237 1.1924 7.0164>, rot=<0.0736 0.1151 0.978 0.1575>}
 Hair_11 {parent=head, loc=<4.1857 4.8292 -6.2859>, rot=<-0.1236 -0.9895 0.0689 0.0276>}
 Hair_12 {parent=head, loc=<1.2661 0.9022 -7.0089>, rot=<-0.0746 -0.9934 0.087 -0.0092>}
 Hair_14 {parent=head, loc=<12.748 -3.1894 6.451>, rot=<0.796 -0.2495 0.3731 -0.406>}
 Hair_15 {parent=head, loc=<14.0034 8.2738 4.3973>, rot=<0.7785 0.3897 0.4534 0.191>}
 Hair_16 {parent=head, loc=<12.1392 10.9008 4.6061>, rot=<0.6757 0.2688 0.656 0.202>}
 Hair_01_c {parent=head, loc=<7.1313 -3.0375 0>, rot=<0.1963 -0.6793 0.6793 -0.1963>}
 Hair_02_c {parent=Hair_01_c, loc=<8.8456 0 0>, rot=<0 -0.0548 0 -0.9985>}
 Hair_03_c {parent=Hair_02_c, loc=<10.2131 0 0>, rot=<0 -0.1225 0 -0.9925>}
 Hair_01_rs {parent=head, loc=<6.652 -2.1417 5.5>, rot=<0.172 -0.6859 0.6859 -0.172>}
 Hair_02_rs {parent=Hair_01_rs, loc=<7.0859 0 0.251>, rot=<0 -0.0548 0 -0.9985>}
 Hair_03_rs {parent=Hair_02_rs, loc=<7.0925 0 0>, rot=<0 0 0 1>}
 Hair_04_rs {parent=Hair_02_rs, loc=<8.5734 -2.9584 -1.8123>, rot=<-0.4264 0.3735 0.3743 -0.7339>}
 Hair_01_ls {parent=head, loc=<6.652 -2.1417 -5.5>, rot=<0.172 -0.6859 0.6859 -0.172>}
 Hair_02_ls {parent=Hair_01_ls, loc=<7.0859 0 0.251>, rot=<0 -0.0548 0 -0.9985>}
 Hair_03_ls {parent=Hair_02_ls, loc=<7.0925 0 0>, rot=<0 0 0 1>}
 Hair_04_ls {parent=Hair_02_ls, loc=<5.4478 0.8598 -3.7527>, rot=<0.3768 0.3195 -0.2806 -0.8229>}
 clavicle_l {parent=spine_05, loc=<4.3575 -0.1466 -2.3223>, rot=<0.0465 -0.7289 -0.009 -0.6829>}
 upperarm_l {parent=clavicle_l, loc=<15.5111 0 0>, rot=<0.0226 -0.3234 -0.0088 -0.9459>}
 upperarm_twist_01_l {parent=upperarm_l, loc=<0.0041 0 0>, rot=<0 0 0 -1>}
 clavicle_r {parent=spine_05, loc=<4.3575 -0.1466 2.3223>, rot=<0.7289 0.0465 -0.6829 0.009>}
 upperarm_r {parent=clavicle_r, loc=<-15.5111 0 0>, rot=<0.0226 -0.3234 -0.0088 -0.9459>}
 upperarm_twist_01_r {parent=upperarm_r, loc=<-0.0041 0 0>, rot=<0 0 0 -1>}
 attach_cape {parent=spine_05, loc=<9.017 -11.714 0>, rot=<0.0616 -0.7044 -0.0616 -0.7044>}
 attach_backpack {parent=spine_05, loc=<-0.123 -13.473 0>, rot=<-0.0093 -0.707 0.0093 -0.707>}
]
