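# Physics asset for the SolarDusk medium male cosmetic's face accessory part, defining hair physics against the upper body.
# Center, left and right chains of box bodies hang from the head with tight angular limits, and both side Hair_04 bodies branch from Hair_02, only the right one locked.
# Head, neck, clavicle and upper spine capsules form a locked-joint body proxy.
# Fully free angular constraints cross-link both side Hair_03 bodies to center Hair_02.
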PhysicsAsset /BRCosmetics/Characters/Player/Male/Medium/Bodies/M_MED_SolarDusk/Meshes/Parts/M_MED_SolarDusk_FaceAcc_Physics
bodies {
 Hair_01_c [{shapeType=Box, extentX=16.3011, extentY=9.2483, extentZ=3.5409, center=<1 0 0>}]
 Hair_01_ls [{shapeType=Box, extentX=16.3187, extentY=9.2483, extentZ=3.5409, center=<1 0 0>}]
 Hair_01_rs [{shapeType=Box, extentX=16.3187, extentY=9.2483, extentZ=3.5409, center=<1 0 0>}]
 Hair_02_c [{shapeType=Box, extentX=16.3011, extentY=9.2483, extentZ=3.5409, center=<1 0 0>}]
 Hair_02_ls [{shapeType=Box, extentX=16.3187, extentY=9.2483, extentZ=3.5409, center=<1 0 0>}]
 Hair_02_rs [{shapeType=Box, extentX=16.3187, extentY=9.2483, extentZ=3.5409, center=<1 0 0>}]
 Hair_03_c [{shapeType=Box, extentX=8.4991, extentY=9.2483, extentZ=3.5409, center=<5 0 0>}]
 Hair_03_ls [{shapeType=Box, extentX=10.6789, extentY=9.6861, extentZ=4.0886, center=<2.4569 0 0>}]
 Hair_03_rs [{shapeType=Box, extentX=10.6789, extentY=9.6861, extentZ=4.0886, center=<2.4569 0 0>}]
 Hair_04_ls [{shapeType=Box, extentX=4.3397, extentY=2.425, extentZ=2.2561, center=<1.1047 0 0>}]
 Hair_04_rs [{shapeType=Box, extentX=4.942, extentY=2.988, extentZ=2.8215, center=<2.0242 0 0>}]
 clavicle_l [{shapeType=Capsule, length=4.627, radius=6.6033, shapeName=clavicle_l_capsule, center=<10.714 0 0>, rotation=<rot 90 180 0>}]
 clavicle_r [{shapeType=Capsule, length=9.8131, radius=6.6033, shapeName=clavicle_r_capsule, center=<-10.714 0 0>, rotation=<rot 90 0 0>}]
 head [{shapeType=Capsule, length=4.5629, radius=5.4683, center=<0 4.6185 0>, rotation=<rot 90 0 0>}]
 neck_01 [{shapeType=Capsule, length=1, radius=1}]
 neck_02 [{shapeType=Capsule, length=1.251, radius=4.5738, rotation=<rot 90 0 0>}]
 spine_04 [{shapeType=Capsule, length=10.4562, radius=9.7566, shapeName=spine_04_capsule, center=<4.1408 0 0>, rotation=<rot 0 -90 0>}]
 spine_05 [{shapeType=Capsule, length=18.221, radius=10.3122, center=<2.0044 1.4021 0>}]
}
constraints [
 {
  bone1Name=head
  bone2Name=neck_02
  swing1LimitDegrees=45
  swing1Motion=Locked
  swing2LimitDegrees=45
  swing2Motion=Locked
  twistLimitDegrees=45
  twistMotion=Locked
 }
 {
  bone1Name=neck_02
  bone2Name=neck_01
  swing1LimitDegrees=45
  swing1Motion=Locked
  swing2LimitDegrees=45
  swing2Motion=Locked
  twistLimitDegrees=45
  twistMotion=Locked
 }
 {
  bone1Name=neck_01
  bone2Name=spine_05
  swing1LimitDegrees=45
  swing1Motion=Locked
  swing2LimitDegrees=45
  swing2Motion=Locked
  twistLimitDegrees=45
  twistMotion=Locked
 }
 {
  bone1Name=clavicle_l
  bone2Name=spine_05
  swing1LimitDegrees=45
  swing1Motion=Locked
  swing2LimitDegrees=45
  swing2Motion=Locked
  twistLimitDegrees=45
  twistMotion=Locked
 }
 {
  bone1Name=clavicle_r
  bone2Name=spine_05
  swing1LimitDegrees=45
  swing1Motion=Locked
  swing2LimitDegrees=45
  swing2Motion=Locked
  twistLimitDegrees=45
  twistMotion=Locked
 }
 {
  bone1Name=spine_05
  bone2Name=spine_04
  swing1LimitDegrees=45
  swing1Motion=Locked
  swing2LimitDegrees=45
  swing2Motion=Locked
  twistLimitDegrees=45
  twistMotion=Locked
 }
 {bone1Name=Hair_04_ls, bone2Name=Hair_02_ls, swing1LimitDegrees=20, swing2LimitDegrees=20, twistLimitDegrees=1}
 {bone1Name=Hair_01_ls, bone2Name=head, swing1LimitDegrees=10, swing2LimitDegrees=20, twistLimitDegrees=10}
 {bone1Name=Hair_02_ls, bone2Name=Hair_01_ls, swing1LimitDegrees=10, swing2LimitDegrees=20, twistLimitDegrees=10}
 {bone1Name=Hair_03_ls, bone2Name=Hair_02_ls, swing1LimitDegrees=10, swing2LimitDegrees=20, twistLimitDegrees=3}
 {bone1Name=Hair_01_rs, bone2Name=head, swing1LimitDegrees=10, swing2LimitDegrees=20, twistLimitDegrees=10}
 {bone1Name=Hair_02_rs, bone2Name=Hair_01_rs, swing1LimitDegrees=10, swing2LimitDegrees=20, twistLimitDegrees=10}
 {bone1Name=Hair_03_rs, bone2Name=Hair_02_rs, swing1LimitDegrees=10, swing2LimitDegrees=20, twistLimitDegrees=3}
 {
  bone1Name=Hair_04_rs
  bone2Name=Hair_02_rs
  swing1LimitDegrees=10
  swing1Motion=Locked
  swing2LimitDegrees=10
  swing2Motion=Locked
  twistLimitDegrees=1
  twistMotion=Locked
 }
 {bone1Name=Hair_03_c, bone2Name=Hair_02_c, swing1LimitDegrees=10, swing2LimitDegrees=30, twistLimitDegrees=3}
 {bone1Name=Hair_01_c, bone2Name=head, swing1LimitDegrees=10, swing2LimitDegrees=20, twistLimitDegrees=10}
 {bone1Name=Hair_02_c, bone2Name=Hair_01_c, swing1LimitDegrees=10, swing2LimitDegrees=20, twistLimitDegrees=10}
 {
  bone1Name=Hair_03_rs
  bone2Name=Hair_02_c
  swing1LimitDegrees=45
  swing1Motion=Free
  swing2LimitDegrees=45
  swing2Motion=Free
  twistLimitDegrees=45
  twistMotion=Free
 }
 {
  bone1Name=Hair_03_ls
  bone2Name=Hair_02_c
  swing1LimitDegrees=45
  swing1Motion=Free
  swing2LimitDegrees=45
  swing2Motion=Free
  twistLimitDegrees=45
  twistMotion=Free
 }
]
